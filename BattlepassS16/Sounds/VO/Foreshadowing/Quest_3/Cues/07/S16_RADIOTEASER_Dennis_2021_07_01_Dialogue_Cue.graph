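# Sound cue for a Dennis radio teaser voice line from the Season 16 battle pass foreshadowing Quest 3.
# A single dialogue player node plays the Dennis dialogue wave with the S16QuestVO speaker and no listed targets.
# It uses the narrator VO sound class and a Season 17 foreshadowing quest VO attenuation asset for distance falloff.
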SoundCue /BattlepassS16/Sounds/VO/Foreshadowing/Quest_3/Cues/07/S16_RADIOTEASER_Dennis_2021_07_01_Dialogue_Cue
root=DialoguePlayer_0
nodes [{
  id=DialoguePlayer_0
  properties {dialogueWaveParameter {
    dialogueWave=/BattlepassS16/Sounds/VO/Foreshadowing/Quest_3/Waves/Dennis/S16_RADIOTEASER_Dennis_2021_07_01_Dialogue
    context {speaker=/BattlepassS16/Sounds/VO/Foreshadowing/S16QuestVO, targets []}
   }}
 }]
properties {
 attenuationSettings=/BattlepassS16/Sounds/VO/Foreshadowing/S17_Foreshadowing_Quest_VO_Att
 maxDistance=5400
 soundClassObject=/Game/Sounds/FortSoundClassesAndModes/Fort_VO_Narrator
 volumeMultiplier=1
}
